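# Vehicle wheel material for the XMatMultichrome cosmetic, with a color-shifting, speed-reactive glow.
# It overrides base color and emissive from MF_WheelParent_Simple, then applies vehicle damage and de-rez transition functions.
# An optional multichrome color hue-shifts a custom color with a time sine and UV gradient.
# Speed relative to car maximum shapes the camera-vector emissive falloff, masked by vertex red; a separate effect blends Fresnel and view terms.
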Material /VehicleCosmetics/Wheels/XmatMultiChrome/Materials/M_Wheel_XMatMultichrome
graph {
 properties {
  bAutomaticallySetUsageInEditor=false
  bUseMaterialAttributes=true
  bUsedWithMeshDeformer=true
  bUsedWithSkeletalMesh=true
  pixelDepthOffsetMode=PDOM_Legacy
  usedShadingModels=Default Lit
 }
 nodes [
  n1 = ScalarParameter() {defaultValue=1, parameterName=GameOBJ_Param:Object_Speed}
  n2 = ScalarParameter() {defaultValue=1, parameterName=GameOBJ_Param:CarMax_Speed}
  Clamp(None=Divide(A=@n1, B=@n2))
  n3 = GetMaterialAttributes(None=MaterialFunctionCall() {materialFunction=&MF_WheelParent_Simple}.MaterialAttributes) {attributeGetTypes [#1, #2]}
  n4 = LinearInterpolate(A=@n3.Emissive Color, B=NamedRerouteUsage(), Alpha=Constant() {r=1})
  n5 = LinearInterpolate(A=@n3.Base Color, B=NamedRerouteUsage(), Alpha=NamedRerouteUsage())
  MaterialFunctionCall(In=MaterialFunctionCall(Material=SetMaterialAttributes!(MaterialAttributes=@n3.MaterialAttributes, Emissive Color=@n4, Base Color=@n5) {attributeSetTypes [#1, #2]}) {materialFunction=&MF_BaseVehicle_Damage}.Result) {materialFunction=&MF_VehicleDeRez_Transition}
  n6 = NamedRerouteUsage()
  n7 = ComponentMask(None=CameraVectorWS()) {b=true}
  n8 = Divide(A=@n1, B=Multiply(A=Constant() {r=0.75}, B=@n2))
  n9 = ConstantBiasScale(None=Abs(None=Clamp(None=@n8))) {bias=0.1, scale=1}
  n10 = Multiply(A=Subtract(A=Constant() {r=1.1}, B=Clamp(None=@n9)), B=Constant() {r=10})
  n11 = Clamp(None=ConstantBiasScale(None=Power(Base=@n7, Exp=@n10)) {bias=0.015, scale=2}) {maxDefault=2}
  n12 = NamedRerouteUsage()
  n13 = VertexColor()
  NamedRerouteDeclaration(None=LinearInterpolate(A=Constant3Vector(), B=Multiply(A=@n6, B=@n11), Alpha=Multiply(A=@n12, B=@n13.R))) {name=Emissive, nodeColor=<rgba 0 1 0.8594 1>}
  n14 = ComponentMask(None=TextureCoordinate()) {g=true}
  n15 = StaticSwitchParameter(True=@n14, False=Multiply(A=@n14, B=Constant() {r=-1})) {defaultValue=true, group=5_Special Effects, parameterName=Custom_Color_Left_Wheel}
  n16 = Sine(None=Multiply(A=Time(), B=ScalarParameter() {defaultValue=0.1, group=5_Special Effects, parameterName=Multuchrome_Speed}))
  n17 = VectorParameter() {group=5_Special Effects, parameterName=CustomColor, defaultValue=<rgba 1 0.85 0 1>}
  n18 = MaterialFunctionCall!(
   Input_Angle=StaticSwitchParameter(True=Add(A=@n15, B=@n16), False=@n16) {defaultValue=true, group=5_Special Effects, parameterName=Custom_Color_Animated_Gradient}
   Input_Color=@n17.RGB
  ) {materialFunction=&/HueShift}
  NamedRerouteDeclaration(None=StaticSwitchParameter(True=Clamp(None=@n18.Result), False=@n17.RGB) {defaultValue=true, group=5_Special Effects, parameterName=Multichrome}) {name=Multichrome, nodeColor=<rgba 0.6172 0 1 1>}
  n19 = TextureSampleParameter2D() {
   group=5_Special Effects
   parameterName=Effect_Masks
   samplerSource=SSM_Wrap_WorldGroupSettings
   samplerType=SAMPLERTYPE_LinearColor
   texture=/VehicleCosmetics/Wheels/XmatMultiChrome/Textures/T_XMatMultiChrome_Tire_RGB
  }
  NamedRerouteDeclaration(None=@n19.R) {name=Mask_RC, nodeColor=<rgba 1 0 0 1>}
  NamedRerouteDeclaration(None=@n19.B) {name=Mask_BC, nodeColor=<rgba 0.0017 0.0479 1 1>}
  n20 = TextureSampleParameter2D() {
   parameterName=RIm_N
   samplerSource=SSM_Wrap_WorldGroupSettings
   samplerType=SAMPLERTYPE_Normal
   texture=/VehicleCosmetics/Wheels/XmatMultiChrome/Textures/T_XMatMultiChrome_Rim_N
  }
  NamedRerouteDeclaration(None=@n20.A) {name=Rim_N_A_CH, nodeColor=<rgba 0 0.3437 1 1>}
  n21 = Clamp(None=MaterialFunctionCall!(Normal, Exponent=Constant() {r=5}) {materialFunction=/VehicleCosmetics/Wheels/Cola_Holographic/Materials/MF_Fresnel_Classic}.Result)
  n22 = Clamp(None=Power(Base=@n7, Exp=Constant() {r=4}))
  n23 = Add(
   A=Clamp(None=ConstantBiasScale(None=@n21) {bias=0.25, scale=1}) {maxDefault=1.5}
   B=Clamp(None=ConstantBiasScale(None=@n22) {bias=0, scale=0.25}) {maxDefault=1.5}
  )
  n24 = NamedRerouteUsage()
  n25 = NamedRerouteUsage()
  n26 = Clamp(None=ConstantBiasScale(None=Clamp(None=@n25)) {bias=-0.1, scale=2})
  n27 = Multiply(A=Clamp(None=NamedRerouteUsage()), B=Multiply(A=@n23, B=LinearInterpolate(A=@n23, B=@n24, Alpha=@n26)))
  NamedRerouteDeclaration(None=@n27) {name=Effect2_BC, nodeColor=<rgba 0.6641 1 0 1>}
  NamedRerouteDeclaration(None=@n13.R) {name=VC_RC, nodeColor=<rgba 1 0 0.9844 1>}
  NamedRerouteDeclaration(None=@n13.G) {name=VC_GC, nodeColor=<rgba 0.051 1 0 1>}
  OneMinus(None=NamedRerouteUsage())
  n28 = Constant() {r=1}
  n29 = MaterialFunctionCall!(Tint=@n28, TintAmount=@n28, Desaturate=Constant()) {materialFunction=/VehicleCosmetics/Wheels/Oldmans/New/Materials/MF_WordENV}
  n30 = Desaturation(None=Multiply(A=@n29.WorldSpace_1, B=@n29.WorldSpace_1)) {luminanceFactors=<rgba 0.3 0.59 0.11 1>}
  n31 = Add(A=@n27, B=@n30)
  QualitySwitch(Default=@n31, Low=@n31, High=@n31, Medium=@n30, Epic=@n31)
 ]
}
