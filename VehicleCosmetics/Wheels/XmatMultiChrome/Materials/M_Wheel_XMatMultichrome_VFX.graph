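# Animated vehicle wheel material for the XMatMultiChrome cosmetic, wrapped in shared damage and de-rez functions.
# Panning texture samples form a distorted, masked Letters layer, with a Left_Wheel switch changing UVs and scroll direction.
# A team-tinted Background layer combines camera-vector falloff with pulsing bump-offset noise.
# The Multichrome layer optionally hue-shifts CustomColor over time or across UVs, and a final lerp blends named-reroute layers into the attributes.
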Material /VehicleCosmetics/Wheels/XmatMultiChrome/Materials/M_Wheel_XMatMultichrome_VFX
graph {
 properties {bAutomaticallySetUsageInEditor=false, bUseMaterialAttributes=true, bUsedWithSkeletalMesh=true, usedShadingModels=Default Lit}
 nodes [
  n1 = LinearInterpolate(A=NamedRerouteUsage(), B=NamedRerouteUsage(), Alpha=NamedRerouteUsage())
  n2 = SetMaterialAttributes!(
   MaterialAttributes
   Base Color=@n1
   Roughness=ScalarParameter() {defaultValue=1, parameterName=Roughness}
   Metallic=ScalarParameter() {parameterName=Metallic}
   Normal=@n1
  ) {attributeSetTypes [
    #1
    #2
    #3
    #4
   ]}
  MaterialFunctionCall(In=MaterialFunctionCall(Material=@n2) {materialFunction=&MF_BaseVehicle_Damage}.Result) {materialFunction=&MF_VehicleDeRez_Transition}
  n3 = Add(
   A=StaticSwitchParameter(True=TextureCoordinate() {uTiling=0.25, vTiling=0.375}, False=TextureCoordinate() {uTiling=-0.25, vTiling=0.375}) {defaultValue=true, parameterName=Left_Wheel}
   B=Constant2Vector() {r=0.32}
  )
  n4 = Time()
  n5 = Constant() {r=0.3}
  n6 = Add(A=TextureCoordinate() {uTiling=0.5, vTiling=0.5}, B=Constant2Vector() {g=0.12})
  n7 = TextureSample(UVs=StaticSwitchParameter(True=Panner(Coordinate=@n6, Time=@n4) {speedX=-0.2}, False=Panner(Time=@n4) {speedX=0.2}) {defaultValue=true, parameterName=Left_Wheel}) {
   samplerSource=SSM_Wrap_WorldGroupSettings
   samplerType=SAMPLERTYPE_LinearColor
   texture=/VehicleCosmetics/Wheels/XmatMultiChrome/Textures/T_XMatMultiChrome_SE
  }
  n8 = TextureSample(UVs=Add(A=Panner(Coordinate=@n3, Time=@n4) {speedX=-0.25, speedY=0.55}, B=Multiply(A=@n5, B=@n7.R))) {
   samplerSource=SSM_Wrap_WorldGroupSettings
   samplerType=SAMPLERTYPE_LinearColor
   texture=&T_XMatMultiChrome_SE
  }
  n9 = ScalarParameter() {parameterName=BlackLetterBlend}
  n10 = Multiply(A=LinearInterpolate(A=@n8.B, B=ConstantBiasScale(None=@n8.B) {bias=0.03, scale=1}, Alpha=@n9), B=@n7.R)
  n11 = ConstantBiasScale(None=@n10) {bias=0, scale=20}
  NamedRerouteDeclaration(None=Clamp(None=LinearInterpolate(A=@n10, B=@n11, Alpha=@n9))) {name=Letters, nodeColor=<rgba 0.7109 0 1 1>}
  n12 = ComponentMask(None=CameraVectorWS()) {g=true}
  n13 = Constant() {r=4}
  n14 = OneMinus(None=Clamp(None=Power(Base=@n12, Exp=@n13)))
  n15 = Clamp(None=ConstantBiasScale(None=Clamp(None=@n14)) {bias=-0.1})
  n16 = TextureCoordinate() {vTiling=-1}
  n17 = BumpOffset(Coordinate=AppendVector(A=ComponentMask(None=@n16) {g=true}, B=ComponentMask(None=@n16) {r=true}), Height=Constant() {r=-0.75})
  n18 = Time()
  n19 = Constant() {r=0.05}
  n20 = Abs(None=Cosine(None=Multiply(A=@n18, B=@n19)))
  n21 = Multiply(
   A=Clamp(None=TextureSample(UVs=@n17) {texture=/VehicleCosmetics/Wheels/_SharedTextures/FX/Universal_Noise_48_RGB}.B) {minDefault=0.02}
   B=Clamp(None=ConstantBiasScale(None=@n20) {bias=0.5, scale=0.75}) {minDefault=0.1}
  )
  n22 = Clamp(None=Multiply(A=Add(A=@n15, B=@n21), B=Constant() {r=0.25}))
  n23 = VectorParameter() {parameterName=TaemColor, defaultValue=<rgba 0 0 1 1>}
  n24 = Max(A=Max(A=@n23.R, B=@n23.R), B=@n23.B)
  n25 = LinearInterpolate(A=Constant3Vector() {constant=<rgba 0.1 0.1 0.1 0>}, B=@n23.RGB, Alpha=Clamp(None=ConstantBiasScale(None=@n24) {bias=-0.1, scale=100}))
  NamedRerouteDeclaration(None=Clamp(None=Multiply(A=@n22, B=@n25))) {name=Background, nodeColor=<rgba 0 1 0.3906 1>}
  n26 = Sine(None=Multiply(A=Time(), B=ScalarParameter() {defaultValue=0.1, parameterName=MultiChrome_Speed}))
  n27 = ComponentMask(None=TextureCoordinate()) {r=true}
  n28 = VectorParameter() {parameterName=CustomColor, defaultValue=<rgba 1 1 0 1>}
  n29 = MaterialFunctionCall!(
   Input_Angle=StaticSwitchParameter(True=Add(A=@n26, B=@n27), False=@n26) {defaultValue=true, parameterName=CustomColor_AnimatedGradient}
   Input_Color=@n28.RGB
  ) {materialFunction=&/HueShift}
  NamedRerouteDeclaration(None=Multiply(
    A=StaticSwitchParameter(True=@n29.Result, False=@n28.RGB) {parameterName=MultiChrome}
    B=ScalarParameter() {defaultValue=1.5, parameterName=EmissiveIntensity}
   )) {name=Multichrome, nodeColor=<rgba 1 0.0703 0 1>}
 ]
}
